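; Server-driven base spawner for a mid-match objective's boss and henchmen, with enable check and spawn entry left as overridable stubs.
; It reserves Living World density, then on request picks a boss patrol point by EQS or random fallback.
; A rift prefab spawns at the ground-traced point; its completion callback releases the reservation and requests the guard spawn.
; Spawned pawns get leashes, offsets and patrol paths, using shuffled paths for henchmen.
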(event EventBeginPlay
  ;[0.03742 0.1562 0.04947] BeginPlay: Reserve density reservoir.
  (|CallParentFunction)
  (Utilities|SwitchNetMode
    (:Server
      (Class|FortAsyncActionCurrentPlaylistReady|PlaylistDataReadyAsync
        (:OnReady
          (bind _benabled (Required|IsObjectiveEnabled))
          (if _benabled
            (Initialize|InitializeSpawner)
            (else
              (Actor|DestroyActor)
              (Fortnite|FortShippingLog self "BeginPlay: Destroying self as Bunker Objective is NOT enabled."))))))))

(event EventEndPlay
  (params (EndPlayReason :type EEndPlayReason%20Enum :default "Destroyed"))
  ;[0.2031 0.108 0.0194] EndPlay: Cleanup delegates and reservoirs.
  (|CallParentFunction EndPlayReason)
  (Utilities|SwitchNetMode
    (:Server
      (LivingWorldManager|UnreserveDensityReservoir)
      (bind _livingworldprefab (Variables|Transient|GetLivingWorldPrefab)
        (:then
          (Default|UnbindEventfromOnSpawn _livingworldprefab (EventDispatchers|CreateEvent :!event HandleRiftSequenceComplete)))
        (:else)))))

(fn BeginGuardSpawnSequenceAtPatrolPoint
  (params (SelectedPatrolPoint :type Fort%20Athena%20Patrol%20Point%20Object%20Reference))
  (bind _returnvalue (Transformation|GetActorRotation SelectedPatrolPoint))
  ;[0.09738 0.04729 0.125] BeginGuardSpawnSequenceAtPatrolPoint: Start the sequence that eventually gets the guard spawning at the desired patrol point.
  (Utilities|SwitchNetMode
    (:Server
      (Utilities|IsValid SelectedPatrolPoint
        (:Is%20Valid
          (Fortnite|FortShippingLog self (Utilities|String|Append "BeginGuardSpawnSequenceAtPatrolPoint: SelectedPatrolPoint is" (Utilities|GetDisplayName SelectedPatrolPoint)))
          (bind (_bsuccess _groundlocation) (GuardSpawn|Protected|TraceGroundLocation SelectedPatrolPoint))
          (if _bsuccess
            (Variables|SetGuardSpawnTransform_Location (Math|Transform|MakeTransform (+ _groundlocation (Variables|Settings|GetGuardSpawnLocationOffset)) _returnvalue))
            (else
              (Fortnite|FortShippingLog self (Utilities|String|Append "BeginGuardSpawnSequenceAtPatrolPoint: Using patrol point directly. Could not find ground location around " (Utilities|GetDisplayName SelectedPatrolPoint)) true)
              (Variables|SetGuardSpawnTransform_Location (Math|Transform|MakeTransform (Transformation|GetActorLocation SelectedPatrolPoint) _returnvalue))))
          (bind _returnvalue_2 (Game|SpawnActorfromClass (Variables|Settings|GetLWP_BossHenchmenRiftClass) (Variables|Transient|GetGuardSpawnTransform) "AlwaysSpawn" "MultiplyWithRoot" 0 0 0))
          (bind _output_get (Variables|Transient|SetLivingWorldPrefab _returnvalue_2))
          (Utilities|IsValid _output_get
            (:Is%20Valid
              (Default|BindEventtoOnSpawn _returnvalue_2 (EventDispatchers|CreateEvent :!event HandleRiftSequenceComplete))
              (Fortnite|FortShippingLog self (Utilities|String|Append "BeginGuardSpawnSequenceAtPatrolPoint: Successfully spawned living world prefab: " (Utilities|GetDisplayName _returnvalue_2))))
            (:Is%20Not%20Valid
              ;;[1 0 0] Warning: LivingWorldPrefab is null.
              (Fortnite|FortShippingLog self "BeginGuardSpawnSequenceAtPatrolPoint: Failed, LivingWorldPrefab is null." true))))
        (:Is%20Not%20Valid
          ;;[1 0 0] Warning: SelectedPatrolPoint is null.
          (Fortnite|FortShippingLog self "BeginGuardSpawnSequenceAtPatrolPoint: SelectedPatrolPoint is null." true))))))

(fn GetAllGuardPawns
  ;[0.04301 0.165 0.1875] Return all the guard pawns spawned.
  (return (Variables|Transient|GetGuardPawns)))

(fn GetBossPatrolPathForPatrolPoint
  (params (PatrolPoint :type Fort%20Athena%20Patrol%20Point%20Object%20Reference))
  ;[0.01291 0.08766 0.08854] Return the patrol path for a specified patrol point.
  (Utilities|FlowControl|Sequence
    (:then_0
      (Utilities|IsValid PatrolPoint
        (:Is%20Valid
          (for _array_element (Variables|Settings|GetBossPatrolPaths)
            (Utilities|IsValid _array_element
              (:Is%20Valid
                (if (Utilities|Array|ContainsItem (Class|FortAthenaPatrolPath|GetPatrolPoints _array_element) PatrolPoint)
                  (return _array_element))))))))
    (:then_1
      (return 0))))

(fn GetBossPawn
  ;[0.02018 0.1615 0.1615] Return the boss pawn.
  (for _array_element (Variables|Transient|GetGuardPawns)
    (Utilities|IsValid _array_element
      (:Is%20Valid
        (if (GameplayTags|HasMatchingGameplayTag _array_element (Variables|Settings|GetBossGameplayTag))
          (return _array_element)))))
  (return 0))

(fn GetBossSpawnPatrolPoints
  ;[0.04743 0.06112 0.1615] Returns the patrol points on which the guards can spawn.
  ; This is expected to be called from the EQS.
  (Utilities|FlowControl|Sequence
    (:then_0
      (for _array_element (Variables|Settings|GetBossPatrolPaths)
        (Utilities|IsValid _array_element
          (:Is%20Valid
            (Utilities|Array|AppendArray (LocalVariables|GetRetPatrolPoints) (Class|FortAthenaPatrolPath|GetPatrolPoints _array_element))))))
    (:then_1
      (return (LocalVariables|GetRetPatrolPoints)))))

(fn GetFallbackBossPatrolPointAndPatrolPath
  ;[0.1302 0.03119 0.1202] Return a random patrol point and patrol path for the boss.
  (Utilities|FlowControl|Sequence
    (:then_0
      (Utilities|SwitchNetMode
        (:Server
          (bind _output_get (LocalVariables|SetPatrolPath (Utilities|Array|RandomArrayItem (Variables|Settings|GetBossPatrolPaths))))
          (Utilities|IsValid _output_get
            (:Is%20Valid
              (return (LocalVariables|GetPatrolPath) (Utilities|Array|RandomArrayItem (Class|FortAthenaPatrolPath|GetPatrolPoints _output_get))))))
        (:Client
          (return 0 0))))))

(fn GetHenchmanPawns
  ;[0.00688 0.06746 0.08333] Return the henchman pawns.
  (for _array_element (Variables|Transient|GetGuardPawns)
    (Utilities|IsValid _array_element
      (:Is%20Valid
        (if (GameplayTags|HasMatchingGameplayTag _array_element (Variables|Settings|GetHenchmanGameplayTag))
          (Utilities|Array|Add (LocalVariables|GetRetPawns) _array_element)))))
  (return (LocalVariables|GetRetPawns)))

(fn GetSpawnerIdTag
  (return (Variables|Settings|GetSpawnerIdTag)))

(fn GetSpawnEventEntry
  (return 0))

(fn HandleGuardSpawnEQSQueryComplete
  (params (QueryInstance :type EQS%20Query%20Instance%20Object%20Reference) (QueryStatus :type EEnvQueryStatus%20Enum :default "Processing"))
  ;[0.02656 0.1327 0.1458] Called when the EQS query is complete.
  (Utilities|SwitchNetMode
    (:Server
      (Utilities|FlowControl|Sequence
        (:then_0
          (Utilities|IsValid QueryInstance
            (:Is%20Valid
              (bind _resultactors (AI|EQS|GetQueryResultsasActors QueryInstance))
              (if (Utilities|Array|IsValidIndex _resultactors)
                (Utilities|IsValid (Utilities|Array|Get(acopy) _resultactors)
                  (:Is%20Valid
                    (bind _asfort_athena_patrol_point (Utilities|Casting|CastToFortAthenaPatrolPoint (Utilities|Array|Get(acopy) _resultactors))
                      (:then
                        (Variables|Transient|SetBossChosenPatrolPoint _asfort_athena_patrol_point)
                        (bind _patrolpath (Helpers|GetBossPatrolPathforPatrolPoint _asfort_athena_patrol_point))
                        (bind _output_get (Variables|Transient|SetBossChosenPatrolPath _patrolpath))
                        (Utilities|IsValid _output_get
                          (:Is%20Not%20Valid
                            ;;[1 0 0] Warning: ChosenPatrolPath is null.
                            (Fortnite|FortShippingLog self "HandleGuardSpawnEQSQueryComplete: BossChosenPatrolPath is null." true))))
                      (:CastFailed
                        ;;[1 0 0] Warning: ResultActor is not a patrol point.
                        (Fortnite|FortShippingLog self (Utilities|String|Append "HandleGuardSpawnEQSQueryComplete: ResultActor is NOT a patrol point. " (Utilities|GetDisplayName (Utilities|Array|Get(acopy) _resultactors))) true))))
                  (:Is%20Not%20Valid
                    ;;[1 0 0] Warning: ResultActor at index 0 is null.
                    (Fortnite|FortShippingLog self "HandleGuardSpawnEQSQueryComplete: ResultActor at index 0 is null." true)))
                (else
                  ;;[1 0 0] Warning: ResultActors is empty.
                  (Fortnite|FortShippingLog self "HandleGuardSpawnEQSQueryComplete: ResultActors is empty." true))))
            (:Is%20Not%20Valid
              ;;[1 0 0] Warning: Query is null.
              (Fortnite|FortShippingLog self "HandleGuardSpawnEQSQueryComplete: Query is null." true))))
        (:then_1
          ;;[0.05868 0.02207 0.07812] Fallback method in case the EQS didn't have valid results.
          (Variables|Transient|GetBossChosenPatrolPath
            (:then
              (Variables|Transient|GetBossChosenPatrolPoint
                (:then)
                (:else
                  _)))
            (:else
              _))
          (Fortnite|FortShippingLog self "HandleGuardSpawnEQSQueryComplete: Using fallback method cause either BossChosenPatrolPoint or BossChosenPatrolPath is null." true)
          (bind (_path _point) (Helpers|GetFallbackBossPatrolPointandPatrolPath))
          (Variables|Transient|SetBossChosenPatrolPath _path)
          (Variables|Transient|SetBossChosenPatrolPoint _point))
        (:then_2
          (Variables|Transient|GetBossChosenPatrolPath
            (:then
              (bind _bosschosenpatrolpoint (Variables|Transient|GetBossChosenPatrolPoint)
                (:then
                  (Fortnite|FortShippingLog self (Utilities|String|Append "HandleGuardSpawnEQSQueryComplete: Success: BossChosenPatrolPoint: " (Utilities|GetDisplayName (Variables|Transient|GetBossChosenPatrolPoint)) " ; BossChosenPatrolPath: " (Utilities|GetDisplayName (Variables|Transient|GetBossChosenPatrolPath))))
                  (GuardSpawn|Protected|BeginGuardSpawnSequenceatPatrolPoint _bosschosenpatrolpoint))
                (:else
                  ;;[1 0 0] Warning: ChosenPatrolPoint is null.
                  (Fortnite|FortShippingLog self "HandleGuardSpawnEQSQueryComplete: Failed, BossChosenPatrolPoint is null even after using fallback." true))))
            (:else
              ;;[1 0 0] Warning: ChosenPatrolPath is null.
              (Fortnite|FortShippingLog self "HandleGuardSpawnEQSQueryComplete: Failed, BossChosenPatrolPath is null even after using fallback." true))))))))

(fn HandleGuardsSpawned
  (params (SpawnedActors :type Array%20of%20Actor%20Object%20References) (bSuccess :type Boolean))
  ;[0.01366 0.1083 0.1094] HandleGuardsSpawned: Callback when all the guards have spawned.
  (Utilities|SwitchNetMode
    (:Server
      (Fortnite|FortShippingLog self (Utilities|String|Append "HandleGuardsSpawned: " (select bSuccess "Success" "Unsuccessful.")) (not bSuccess))
      (for _array_element SpawnedActors
        (bind _asfort_player_pawn_athena (Utilities|Casting|CastToFortPlayerPawnAthena _array_element))
        (Utilities|Array|Add (Variables|Transient|GetGuardPawns) _asfort_player_pawn_athena)
        (Fortnite|FortShippingLog self (Utilities|String|Append "HandleGuardSpawned: Pawn spawned successfully: " (Fortnite|GetPlayerNameSafe _asfort_player_pawn_athena)))
        (Helpers|SetupLeashforPawn _asfort_player_pawn_athena self (Variables|Settings|GetLeashRadiusInner) (Variables|Settings|GetLeashRadiusOuter))
        (Transformation|AddActorWorldOffset _asfort_player_pawn_athena (Math|Conversions|ToVector(Vector2D) (Math|Conversions|ToVector2D(Vector) (Math|Vector|RotateVectorAroundAxis (* (Math|Vector|GetForwardVector (.rotation (Variables|Transient|GetGuardSpawnTransform))) (Variables|Settings|GetInterGuardOffsetDistance)) (Math|Float|SafeDivide 3.6e+02 (Math|Conversions|ToFloat(Integer) (Utilities|Array|Length (Variables|GetSpawnedActors)))) "0.000000,0.000000,1.000000"))) true false)
        (bind _bosschosenpatrolpath (Variables|Transient|GetBossChosenPatrolPath)
          (:then
            (if (Utilities|Array|IsValidIndex (Class|FortAthenaPatrolPath|GetPatrolPoints _bosschosenpatrolpath) (Utilities|Array|LastIndex (Class|FortAthenaPatrolPath|GetPatrolPoints _bosschosenpatrolpath)))
              (Utilities|IsValid (Utilities|Array|Get(acopy) (Class|FortAthenaPatrolPath|GetPatrolPoints _bosschosenpatrolpath) (Utilities|Array|LastIndex (Class|FortAthenaPatrolPath|GetPatrolPoints _bosschosenpatrolpath)))
                (:Is%20Valid
                  (bind (_returnvalue_roll _returnvalue_pitch) (Math|Rotator|BreakRotator (Transformation|GetActorRotation _asfort_player_pawn_athena)))
                  (bind (_returnvalue_roll_2 _returnvalue_pitch_2 _returnvalue_yaw) (Math|Rotator|BreakRotator (Math|Rotator|FindLookatRotation (Transformation|GetActorLocation _asfort_player_pawn_athena) (Transformation|GetActorLocation (Utilities|Array|Get(acopy) (Class|FortAthenaPatrolPath|GetPatrolPoints _bosschosenpatrolpath) (Utilities|Array|LastIndex (Class|FortAthenaPatrolPath|GetPatrolPoints _bosschosenpatrolpath)))))))
                  (Transformation|SetActorRotation _asfort_player_pawn_athena (Math|Rotator|MakeRotator _returnvalue_roll _returnvalue_pitch _returnvalue_yaw) false))
                (:Is%20Not%20Valid
                  _))
              ;;[0.0134 0.1113 0.125] If desired, remove the NPC from the current group.
              (if (Variables|Settings|GetbDisableGroup)
                (Utilities|IsValid (Actor|GetComponentbyClass _asfort_player_pawn_athena "/Script/FortniteAI.FortPawnComponent_AIGroup")
                  (:Is%20Valid
                    (AIGroup|RemovefromCurrentGroup (Actor|GetComponentbyClass _asfort_player_pawn_athena "/Script/FortniteAI.FortPawnComponent_AIGroup")))))
              (else
                _)))
          (:else
            _)))
      (GuardSpawn|Protected|SetupSpawnedGuardsPatrolPaths)
      (Default|CallOnGuardsSpawned self))))

(fn HandleRiftSequenceComplete
  (bind _returnvalue (Class|FortAthenaLivingWorldManager|GetLivingWorldManager))
  ;[0.01713 0.1422 0.1562] HandleRiftSequenceComplete: Called when the rift sequence has been completed at the desired patrol point, spawn in the guard!
  (Utilities|SwitchNetMode
    (:Server
      (Utilities|FlowControl|Sequence
        (:then_0
          ;;[0.0596 0.03179 0.125] Unbind delegate.
          (bind _livingworldprefab (Variables|Transient|GetLivingWorldPrefab)
            (:then
              (Default|UnbindEventfromOnSpawn _livingworldprefab (EventDispatchers|CreateEvent :!event HandleRiftSequenceComplete)))
            (:else)))
        (:then_1
          (Utilities|IsValid _returnvalue
            (:Is%20Valid
              (LivingWorldManager|UnreserveDensityReservoir)
              (bind (_requestid _returnvalue_2) (Class|FortAthenaLivingWorldManager|TrySpawnEvent _returnvalue (Required|GetSpawnEventEntry) (Variables|Transient|GetGuardSpawnTransform) (EventDispatchers|CreateEvent :!event HandleGuardsSpawned)))
              (Variables|Transient|SetRequestId _requestid)
              (if _returnvalue_2
                (Fortnite|FortShippingLog self "HandleRiftSequenceComplete: TrySpawnEvent Request Created Successfully.")
                (else
                  ;;[1 0 0] Warning: Failed to create request.
                  (Fortnite|FortShippingLog self "HandleRiftSequenceComplete: TrySpawnEvent Failed To Create Request." true))))
            (:Is%20Not%20Valid
              ;;[1 0 0] Warning: LivingWorldManager is invalid.
              (Fortnite|FortShippingLog self "HandleRiftSequenceComplete: LivingWorldManager is invalid." true))))))))

(fn InitializeSpawner
  (Utilities|SwitchNetMode
    (:Server
      (LivingWorldManager|ReserveDensityReservoir)
      (Utilities|Array|Shuffle (Variables|Settings|GetHenchmanPatrolPaths)))))

(fn IsObjectiveEnabled
  (return false))

(fn RequestSpawnGuard
  (params (RequestingActor :type Actor%20Object%20Reference))
  ;[0.01185 0.09375 0.08252] RequestSpawnGuard: Randomly choose a location on a patrol path and start spawning the NPC there.
  (Utilities|SwitchNetMode
    (:Server
      (if (not (Variables|Transient|GetbRequestedSpawn))
        (Variables|Transient|SetbRequestedSpawn true)
        (Variables|Transient|SetSpawnGuardRequestingActor RequestingActor)
        (bind _returnvalue (AI|EQS|RunEQSQuery (Variables|Required|GetEnvQuery_GuardSpawnPoint) self "RandomBest25Pct"))
        (Utilities|IsValid _returnvalue
          (:Is%20Valid
            (Variables|Transient|SetEQSQueryInstance_GuardSpawnPoint _returnvalue)
            (AI|EQS|BindEventtoOnQueryFinished _returnvalue (EventDispatchers|CreateEvent :!event HandleGuardSpawnEQSQueryComplete))
            (Fortnite|FortShippingLog self "RequestSpawnGuard: EQS requested successfully."))
          (:Is%20Not%20Valid
            ;;[1 0 0] Warning: EQS failed.
            (Fortnite|FortShippingLog self "RequestSpawnGuard: EQS Query failed to run, using fallback method." true)
            ;;[0.07212 0.178 0.2292] Fallback method if EQS fails.
            (bind (_path _point) (Helpers|GetFallbackBossPatrolPointandPatrolPath))
            (Variables|Transient|SetBossChosenPatrolPath _path)
            (Variables|Transient|SetBossChosenPatrolPoint _point)
            (Variables|Transient|GetBossChosenPatrolPath
              (:then
                (bind _bosschosenpatrolpoint (Variables|Transient|GetBossChosenPatrolPoint)
                  (:then
                    (GuardSpawn|Protected|BeginGuardSpawnSequenceatPatrolPoint _bosschosenpatrolpoint))
                  (:else
                    ;;;[1 0 0] Warning: Fallback failed.
                    (Fortnite|FortShippingLog self "RequestSpawnGuard: Fallback method failed as ChosenBossPatrolPoint  is null." true))))
              (:else
                ;;;[1 0 0] Warning: Fallback failed.
                (Fortnite|FortShippingLog self "RequestSpawnGuard: Fallback method failed as BossChosenPatrolPath is null." true)))))
        (else
          ;;[1 0 0] Warning: Already requested.
          (Fortnite|FortShippingLog self "RequestSpawnGuard: Failed, already requested." true))))))

(fn ReserveDensityReservoir
  (bind _returnvalue (Class|FortAthenaLivingWorldManager|GetLivingWorldManager))
  ;[0.1135 0.04086 0.125] ReserveDensityReservoir: Ensures that the there's room in the living world manager to spawn the NPCs.
  (Utilities|SwitchNetMode
    (:Server
      (Utilities|IsValid _returnvalue
        (:Is%20Valid
          (Class|FortAthenaLivingWorldManager|RegisterDensityReservoir _returnvalue self (Variables|Settings|GetReservoirDensity) (Variables|Settings|GetReservoirRange))
          (Fortnite|FortShippingLog self "ReserveDensityReservoir: Successfully registered density reservoir."))
        (:Is%20Not%20Valid
          ;;[1 0 0] Warning: LivingWorldManager is null.
          (Fortnite|FortShippingLog self "ReserveDensityReservoir: LivingWorldManager is null." true))))))

(fn SetupLeashForPawn
  (params (Pawn :type Pawn%20Object%20Reference) (ActorToFollow :type Actor%20Object%20Reference) (InnerRadius :type Float%20%28single-precision%29) (OuterRadius :type Float%20%28single-precision%29))
  ;[0.1105 0.03767 0.125] SetupLeashForPawn: Setup the leash settings for the input pawn.
  (Utilities|SwitchNetMode
    (:Server
      (Utilities|IsValid Pawn
        (:Is%20Valid
          (bind _asfort_athena_aibot_controller (Utilities|Casting|CastToFortAthenaAIBotController (Pawn|GetController Pawn))
            (:then
              (Class|FortAthenaAIBotController|SetLeashActortoFollow :self _asfort_athena_aibot_controller :ActorToFollow ActorToFollow :LeashLocationUpdateRate 1.0)
              (Class|FortAthenaAIBotController|SetRadialLeashInnerRadius _asfort_athena_aibot_controller InnerRadius)
              (Class|FortAthenaAIBotController|SetRadialLeashOuterRadius _asfort_athena_aibot_controller OuterRadius))
            (:CastFailed
              ;;[1 0 0] Warning: Pawn does NOT use the expected controller
              (Fortnite|FortShippingLog self (Utilities|String|Append "HandlePawnSpawned: Pawn does NOT used controller FortAthenaAIBotController: ") true))))))))

(fn SetupPatrolForPawn
  (params (Pawn :type Fort%20Player%20Pawn%20Athena%20Object%20Reference) (NewPatrolPath :type Fort%20Athena%20Patrol%20Path%20Object%20Reference))
  ;[0.06029 0.1091 0.1302] SetupPatrolForPawn: Setup the patrol path for the input pawn.
  (Utilities|SwitchNetMode
    (:Server
      (Utilities|IsValid Pawn
        (:Is%20Valid
          (Utilities|IsValid NewPatrolPath
            (:Is%20Valid
              (bind _asfort_athena_aibot_controller (Utilities|Casting|CastToFortAthenaAIBotController (Pawn|GetController Pawn)))
              (bind _cachedpatrollingcomponent (Variables|/Script/FortniteAI.FortAthenaAIBotController|GetCachedPatrollingComponent _asfort_athena_aibot_controller)
                (:then
                  (Class|FortAthenaNpcPatrollingComponent|SetPatrolPath _cachedpatrollingcomponent NewPatrolPath false)
                  (Fortnite|FortShippingLog self (Utilities|String|Append "SetupPatrolForPawn: Patrolling setup successfully for pawn: " (Utilities|GetDisplayName Pawn) " to " (Utilities|GetDisplayName NewPatrolPath))))
                (:else
                  ;;[1 0 0] Warning: PatrollingComponent is null.
                  (Fortnite|FortShippingLog self "SetupPatrolForPawn: CachedPatrollingComponent is null" true))))
            (:Is%20Not%20Valid
              ;;[1 0 0] Warning: Patrol path is null.
              (Fortnite|FortShippingLog self (Utilities|String|Append "SetupPatrolForPawn: Patrol path is invalid for Pawn: " (Utilities|GetDisplayName Pawn)) true))))))))

(fn SetupSpawnedGuardsPatrolPaths
  (bind _henchmanpatrolpaths (Variables|Settings|GetHenchmanPatrolPaths))
  ;[0.0175 0.1105 0.1458] SetupSpawnedGuardsPatrolPaths: Setup the patrol path for the spawned NPC guards.
  (Utilities|SwitchNetMode
    (:Server
      (bind _bosspawn (Helpers|GetBossPawn))
      (Helpers|SetupPatrolforPawn _bosspawn (Variables|Transient|GetBossChosenPatrolPath))
      (bind _henchmanpawns (Helpers|GetHenchmanPawns))
      (for (_array_element _array_index) _henchmanpawns
        (if (Utilities|Array|IsValidIndex _henchmanpatrolpaths _array_index)
          (Utilities|IsValid (Utilities|Array|Get(acopy) _henchmanpatrolpaths _array_index)
            (:Is%20Valid
              (Helpers|SetupPatrolforPawn _array_element (Utilities|Array|Get(acopy) _henchmanpatrolpaths _array_index)))
            (:Is%20Not%20Valid
              ;;[1 0 0] Warning: HenchmanPatrolPath is NULL at index.
              (Fortnite|FortShippingLog self (Utilities|String|Append "HandleGuardsSpawned: Henchman Patrol Point is null at index : " (Utilities|String|ToString(Integer) _array_index)) true)))
          (else
            ;;[1 0 0] Warning: HenchmanPatrolPath is invalid at index.
            (Fortnite|FortShippingLog self (Utilities|String|Append "HandleGuardsSpawned: Henchman Patrol Point is invalid at index : " (Utilities|String|ToString(Integer) _array_index)) true)))))))

(fn TraceGroundLocation
  (params (SelectedPatrolPoint :type Actor%20Object%20Reference))
  ;[0.04109 0.1306 0.1615] Returns the ground location around the patrol point
  (Utilities|FlowControl|Sequence
    (:then_0
      (Utilities|SwitchNetMode
        (:Server
          (Utilities|IsValid SelectedPatrolPoint
            (:Is%20Valid
              (bind (_outhit _returnvalue) (Collision|LineTraceByProfile (+ (Transformation|GetActorLocation SelectedPatrolPoint) (Variables|Settings|GetGroundTraceStartOffset)) (+ (Transformation|GetActorLocation SelectedPatrolPoint) (Variables|Settings|GetGroundTraceEndOffset)) "FortSupplyDropGroundTrace"))
              (return _returnvalue (.ImpactPoint (Collision|BreakHitResult _outhit))))))))
    (:then_1
      (return false "0, 0, 0"))))

(fn UnreserveDensityReservoir
  (bind _returnvalue (Class|FortAthenaLivingWorldManager|GetLivingWorldManager))
  ;[0.01362 0.1082 0.1354] UnregisterDensityReservoir: No need to reserve room for the NPC anymore.
  (if (Networking|IsDedicatedServer)
    (Utilities|IsValid _returnvalue
      (:Is%20Valid
        (Class|FortAthenaLivingWorldManager|UnregisterDensityReservoir _returnvalue self)
        (Fortnite|FortShippingLog self "UnregisterDensityReservoir: Successfully unregistered density reservoir."))
      (:Is%20Not%20Valid
        ;;[1 0 0] Warning: LivingWorldManager is null.
        (Fortnite|FortShippingLog self "UnregisterDensityReservoir: LivingWorldManager is null." true)))))

(fn ConstructionScript
  ;[0.05125 0.232 0.2344] Construction: Update the leash sphere radius
  (|CallParentFunction)
  (bind _leashsphere_editoronly (Variables|Default|GetLeashSphere_EDITORONLY)
    (:then
      (Components|Sphere|SetSphereRadius _leashsphere_editoronly (Variables|Settings|GetLeashRadiusOuter) false))
    (:else)))
