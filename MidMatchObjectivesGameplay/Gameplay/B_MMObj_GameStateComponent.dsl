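; Base game state component for mid-match objectives, with empty ActivateObjective and ResetObjective functions.
; On the server, if enabled, it tracks the safe zone indicator through the event router and activates at a hotfixed phase when the configured state equals Holding, or by deferred activation, after a random delay.
; It also provides delayed RepNotify announcements, quest events for players without a placement, and data registry scope from the indicator's minigame.
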(event EventBeginPlay
  ;[0 0.2865 0.01994] BeginPlay: On the server, bind to know when the safe zone indicator has spawned.
  (|CallParentFunction)
  (Networking|SwitchHasAuthority
    (:Authority
      (Class|FortAsyncActionCurrentPlaylistReady|PlaylistDataReadyAsync
        (:OnReady
          (if (CallFunction|IsObjectiveEnabled)
            (bind _returnvalue (GameplayEventRouter|GetEventRouter "/Script/FortniteGame.FortGameState" (Components|GetOwner)))
            (bind (_asynctaskproxy _payload) (GameplayEventRouter|StartListeningToStatefulEvent _returnvalue "/Script/FortniteGame.FortSafeZoneIndicatorUpdatedEvent")
              (:then
                (Variables|Default|SetSafeZoneIndicatorUpdatedAsyncTask _asynctaskproxy))
              (:OnEventReceived
                _)
              (:OnSavedEventStateExists
                _))
            (Default|HandleSafeZoneIndicatorSpawned (Utilities|Struct|BreakFortSafeZoneIndicatorUpdatedEvent _payload))))))))

(event EventEndPlay
  (params (EndPlayReason :type EEndPlayReason%20Enum :default "Destroyed"))
  ;[0.1823 0.1237 0.03038] EndPlay: Unbind delegates.
  (Utilities|FlowControl|Sequence
    (:then_0
      (Networking|SwitchHasAuthority
        (:Authority
          (CallFunction|UnbindSafeZoneDelegates)
          (Utilities|Time|ClearandInvalidateTimerbyHandle (Variables|Transient|GetActivateObjectiveTimerHandle)))))
    (:then_1
      (|CallParentFunction EndPlayReason))))

(event DeferredActivation|EventTryDeferredActivate
  (params (ActivationSafeZonePhase :type Integer) (ActivationSafeZoneState :type EFortSafeZoneState%20Enum))
  (CallFunction|HandleDelayedActivation))

(fn ActivateObjective)

(fn BroadcastAnnouncement
  (bind _announcementinfo (Variables|Announcement|GetAnnouncementInfo))
  (bind (_enabled_13_eb6e96924bb94c120001ae91008e1ca7 _delay_16_5115e9a54db74b7958337ab70683c3eb) (Utilities|Struct|BreakMMObjAnnouncementInfoStruct _announcementinfo))
  (bind _returnvalue (ScalableFloat|GetValueatLevel _delay_16_5115e9a54db74b7958337ab70683c3eb))
  ;[0.04779 0.1834 0.2083] BroadcastAnnouncement: Replicate the announcement variable after the desired delay.
  (Utilities|SwitchHasAuthority
    (:Authority
      (if (ScalableFloat|GetValueasBool _enabled_13_eb6e96924bb94c120001ae91008e1ca7)
        (if (> _returnvalue 0)
          (Utilities|Time|SetTimerbyEvent (EventDispatchers|CreateEvent :!event ReplicateAnnouncement) _returnvalue)
          (else
            (Announcement|ReplicateAnnouncement)))))))

(fn BuildDataRegistryResolverScope
  (bind _safezoneindicator (Variables|Default|GetSafeZoneIndicator))
  (bind _returnvalue (Minigame|GetMinigame _safezoneindicator :!class /Script/FortniteGame.MinigameBlueprintLibrary))
  ;[1 0.8475 0.2776] Gives the name of the volume as a scope to the data registry. Needed for calls to GetValueWithContext to work.
  ; Context Ref is grabbed before a pull from the data registry is made.
  (Utilities|IsValid _safezoneindicator
    (:Is%20Valid
      (Utilities|IsValid _returnvalue
        (:Is%20Valid
          (return (Utilities|Array|MakeArray (SpatialGameplay|GetSpatialGameplayInterfaceFNameforActor _returnvalue)) 0 true))
        (:Is%20Not%20Valid
          _)))
    (:Is%20Not%20Valid
      _))
  (return 0 0 false))

(fn GetCacheContextOverride
  ;[0.4163 1 0.9801] We don't need to return anything here since we don't need to override the Cache Context.
  (return 0))

(fn HandleDelayedActivation
  ;[0.6362 0.1696 0.8906] Don't unbind the events until EndPlay() in Creative.
  (if (not (or (ScalableFloat|GetValueasBoolwithContext (Variables|Default|GetHotfix_InCreativeRoyale)) (Variables|Default|GetbSkipUnbindSafeZoneDelegates)))
    (CallFunction|UnbindSafeZoneDelegates))
  (bind _output_get (LocalVariables|SetRandomDelay (Math|Random|RandomFloatinRange (ScalableFloat|GetValueatLevel (Variables|Hotfix|GetHotfix_ActivateDelayMin)) (ScalableFloat|GetValueatLevel (Variables|Hotfix|GetHotfix_ActivateDelayMax)))))
  (if (> _output_get 0)
    (bind _returnvalue (Utilities|Time|SetTimerbyEvent (EventDispatchers|CreateEvent :!event ActivateObjective) _output_get))
    (Variables|Transient|SetActivateObjectiveTimerHandle _returnvalue)
    (else
      (CallFunction|ActivateObjective))))

(fn HandleSafeZoneIndicatorSpawned
  (params (SpawnedSafeZoneIndicator :type Fort%20Safe%20Zone%20Indicator%20Object%20Reference))
  ;[0.02353 0.2078 0.1137] HandleSafeZoneIndicatorSpawned: Bind to know whenever the safe zone state changes.
  (Networking|SwitchHasAuthority
    (:Authority
      (Utilities|IsValid SpawnedSafeZoneIndicator
        (:Is%20Valid
          (SafeZone|BindEventtoSafezoneStateChangedDelegate SpawnedSafeZoneIndicator (EventDispatchers|CreateEvent :!event HandleSafeZoneStateChanged))
          ;[0.08785 0.776 0.4246] Grab this to be the context actor  so FortDataRegistry can be accessed.
          ;
          ; runs when new storm circle is spawned (after begin play)
          (Variables|Default|SetSafeZoneIndicator SpawnedSafeZoneIndicator)
          (Fortnite|FortShippingLog self "HandleSafeZoneIndicatorSpawned: Spawned.")
          (Default|HandleSafeZoneUpdated (SafeZone|GetCurrentPhaseIndex SpawnedSafeZoneIndicator) (SafeZone|GetSafeZoneState SpawnedSafeZoneIndicator)))))))

(fn HandleSafeZoneStateChanged
  (params (SafeZoneIndicator :type Fort%20Safe%20Zone%20Indicator%20Object%20Reference) (SafeZoneState :type EFortSafeZoneState%20Enum))
  ;[0.02663 0.1667 0.06633] HandleSafeZoneStateChanged: Call the function that takes action based on safe zone state changes.
  (Networking|SwitchHasAuthority
    (:Authority
      (Utilities|IsValid SafeZoneIndicator
        (:Is%20Valid
          (Default|HandleSafeZoneUpdated (SafeZone|GetCurrentPhaseIndex SafeZoneIndicator) SafeZoneState))))))

(fn HandleSafeZoneUpdated
  (params (PhaseIndex :type Integer) (SafeZoneState :type EFortSafeZoneState%20Enum))
  ;[0.05759 0.0169 0.08333] HandleSafeZoneUpdated: Based on the safe zone state, take action!
  (Networking|SwitchHasAuthority
    (:Authority
      (if (not (DeferredActivation|ShouldUseDeferredActivation))
        (Fortnite|FortShippingLog self (Utilities|String|Append "HandleSafeZoneUpdated: Phase Index is " (Utilities|String|ToString(Integer) PhaseIndex) " ; SafeZoneState is " (Utilities|String|EnumtoString SafeZoneState)))
        (if (and (== (ScalableFloat|GetValueasInteger (Variables|Hotfix|GetHotfix_ActivateSafeZonePhaseIndex)) PhaseIndex) (Utilities|Enum|Equal(Enum) (Utilities|Enum|BytetoEnumEFortSafeZoneState (Math|Conversions|ToByte(Integer) (ScalableFloat|GetValueasInteger (Variables|Hotfix|GetHotfix_ActivateSafeZoneState)))) "Holding"))
          (CallFunction|HandleDelayedActivation))))))

(fn IsObjectiveEnabled
  ;[0.02845 0.1088 0.125] Returns if the objective is enabled.
  (return (ScalableFloat|GetValueasBool (Variables|Hotfix|GetHotfix_ObjectiveGameplayEnabled))))

(fn OnRep_bBroadcastAnnouncement
  (Default|CallOnBroadcastAnnouncement (Variables|Announcement|GetAnnouncementInfo)))

(fn ReplicateAnnouncement
  ;[0.03745 0.1188 0.1615] ReplicateAnnouncement: Rep the announcement to the clients.
  (Networking|SwitchHasAuthority
    (:Authority
      (Variables|Announcement|SetbBroadcastAnnouncement true))))

(fn ResetObjective)

(fn SendQuestEventToAllPlayers
  (params (QuestTag :type Gameplay%20Tag%20Structure))
  ;[0.008789 0.1406 0.1406] SendQuestEventToAllPlayers: Helper function to send quest events to all players.
  (Networking|SwitchHasAuthority
    (:Authority
      (if (GameplayTags|IsGameplayTagValid QuestTag)
        (for _array_element (Fortnite|GetAllFortPlayerControllers true)
          (Utilities|IsValid _array_element
            (:Is%20Valid
              (bind _asfort_player_state_athena (Utilities|Casting|CastToFortPlayerStateAthena (Variables|/Script/Engine.Controller|GetPlayerState _array_element)))
              (if (<= (Variables|/Script/FortniteGame.FortPlayerStateAthena|GetPlace _asfort_player_state_athena) 0)
                (Utilities|IsValid (Quests|GetQuestManager _array_element "Invalid")
                  (:Is%20Valid
                    (Quest|SendComplexCustomStatEvent (Quests|GetQuestManager _array_element "Invalid") 0 (GameplayTags|MakeGameplayTagContainerfromTag QuestTag) "(GameplayTags=)" 1)
                    (Ability|SendGameplayEventtoActor _array_element QuestTag)))))))))))

(fn UnbindSafeZoneDelegates
  (bind _safezoneindicator (Variables|Default|GetSafeZoneIndicator))
  ;[0.04305 0.1615 0.1388] UnbindSafeZoneDelegates: Unbind all safe zone related delegates.
  (Networking|SwitchHasAuthority
    (:Authority
      (bind _safezoneindicatorupdatedasynctask (Variables|Default|GetSafeZoneIndicatorUpdatedAsyncTask)
        (:then
          (GameplayEventRouter|StopListeningtoStatefulEvent _safezoneindicatorupdatedasynctask))
        (:else
          _))
      (Utilities|IsValid _safezoneindicator
        (:Is%20Valid
          (SafeZone|UnbindEventfromSafezoneStateChangedDelegate _safezoneindicator (EventDispatchers|CreateEvent :!event HandleSafeZoneStateChanged)))))))
