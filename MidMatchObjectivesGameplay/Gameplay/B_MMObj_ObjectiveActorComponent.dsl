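; Actor component for mid-match objective actors that registers its owner for automatic marking on players.
; Registration runs once on the dedicated server, waiting for playlist data before checking hotfixable ScalableFloat enable settings.
; It passes the owner and marker settings to the objective marker controller component on every existing player controller.
; Unregistration removes the owner from those components and also runs on EndPlay.
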(event Custom|RegisterObjectiveForAutoMarking
  (bind (_markingenabled_37_e9c710f04f16fc54efb91c86c40d0196 _markfrequnecy_27_9396cf96491030a734e4d7ac9f77d57a _markdistance_28_800e90424cd7c2d710d2f493e83988ca _maxmarkcount_30_c46e92db43b1c35641a947b346b1541b _includesquad_34_374997384aba530512d186b0137105f9 _usehoveredmarkerdetail_35_1dd554b84ad257830f7aab9a215d2479) (Utilities|Struct|BreakMMObjObjectiveMarkerHotfixInfoStruct (Variables|Settings|GetMarkerComponentInfo)))
  ;[0 0.2865 0.007328] RegisterObjectiveForAutoMarking: Register this objective to be marked for all players that are connected to the game.
  (if (and (Networking|IsDedicatedServer) (not (Variables|Transient|GetbRegisteredForAutoMarking)))
    (Class|FortAsyncActionCurrentPlaylistReady|PlaylistDataReadyAsync
      (:OnReady
        (Fortnite|FortShippingLog self (Utilities|String|ToString(Text) (Utilities|Text|FormatText "INVTEXT(\"RegisterObjectiveForAutoMarking: OwnerActor {OA} ; MarkingEnabled {A} ; MarkFrequency {B} ; MarkDistance {C} ; MaxMarkCount {D} ; Include Squad {E} ; UseHoveredMarkerDetail {F}\")" (Utilities|GetDisplayName (Components|GetOwner)) (ScalableFloat|GetValueasBool _markingenabled_37_e9c710f04f16fc54efb91c86c40d0196) (ScalableFloat|GetValueatLevel _markfrequnecy_27_9396cf96491030a734e4d7ac9f77d57a) (ScalableFloat|GetValueatLevel _markdistance_28_800e90424cd7c2d710d2f493e83988ca) (ScalableFloat|GetValueasInteger _maxmarkcount_30_c46e92db43b1c35641a947b346b1541b) (ScalableFloat|GetValueasBool _includesquad_34_374997384aba530512d186b0137105f9) (ScalableFloat|GetValueasBool _usehoveredmarkerdetail_35_1dd554b84ad257830f7aab9a215d2479))))
        (if (and (ScalableFloat|GetValueasBool (Utilities|Struct|BreakMMObjObjectiveMarkerHotfixInfoStruct (Variables|Settings|GetMarkerComponentInfo))) (ScalableFloat|GetValueasBool (Variables|Settings|GetObjectiveMarkingEnabled)))
          (Fortnite|FortShippingLog self (Utilities|String|Append "RegisterObjectiveForAutoMarking: Registering actor: " (Utilities|GetDisplayName (Components|GetOwner))))
          (Variables|Transient|SetbRegisteredForAutoMarking true)
          (for _array_element (Fortnite|GetAllFortPlayerControllers true)
            (Utilities|IsValid (Actor|GetComponentbyClass _array_element "/MidMatchObjectivesGameplay/Gameplay/B_MMObj_ObjectiveMarkerControllerComponent.B_MMObj_ObjectiveMarkerControllerComponent_C")
              (:Is%20Valid
                (Public|RegisterObjective (Actor|GetComponentbyClass _array_element "/MidMatchObjectivesGameplay/Gameplay/B_MMObj_ObjectiveMarkerControllerComponent.B_MMObj_ObjectiveMarkerControllerComponent_C") (Components|GetOwner) (Variables|Settings|GetMarkerComponentInfo))))))))))

(event Custom|UnregisterObjectiveForAutoMarking
  ;[0.1406 0.08044 0.01577] UnregisterObjectiveForAutomarking: Unregister from the objective marker controller component.
  (if (and (Networking|IsDedicatedServer) (Variables|Transient|GetbRegisteredForAutoMarking))
    (if (ScalableFloat|GetValueasBool (Utilities|Struct|BreakMMObjObjectiveMarkerHotfixInfoStruct (Variables|Settings|GetMarkerComponentInfo)))
      (Fortnite|FortShippingLog self (Utilities|String|Append "UnregisterObjectiveForAutoMarking: Unregistering actor: " (Utilities|GetDisplayName (Components|GetOwner))))
      (for _array_element (Fortnite|GetAllFortPlayerControllers true)
        (Utilities|IsValid (Actor|GetComponentbyClass _array_element "/MidMatchObjectivesGameplay/Gameplay/B_MMObj_ObjectiveMarkerControllerComponent.B_MMObj_ObjectiveMarkerControllerComponent_C")
          (:Is%20Valid
            (Public|UnregisterObjective (Actor|GetComponentbyClass _array_element "/MidMatchObjectivesGameplay/Gameplay/B_MMObj_ObjectiveMarkerControllerComponent.B_MMObj_ObjectiveMarkerControllerComponent_C") (Components|GetOwner))))))))

(event EventEndPlay
  (params (EndPlayReason :type EEndPlayReason%20Enum :default "Destroyed"))
  ;[0.2083 0.1095 0.02093] EndPlay: Unregister for auto marking.
  (|CallParentFunction EndPlayReason)
  (if (Networking|IsDedicatedServer)
    (CallFunction|UnregisterObjectiveforAutoMarking)))
