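; Player controller component that auto-marks registered mid-match objective actors for its owning player, running on the dedicated server.
; RegisterObjective stores hotfixable per-actor settings in a map, and a timer runs UpdateMarking while entries exist.
; Each update marks nearby objectives at intervals, cancels out-of-range markers, and unregisters objectives at maximum mark count or too far outside the safe zone.
; It destroys itself when marking is disabled, and destroyed actors unregister automatically.
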(event EventBeginPlay
  ;[0.01497 0.1042 0.02477] BeginPlay: Check if objective marking is enabled.
  (|CallParentFunction)
  (if (Networking|IsDedicatedServer)
    (Class|FortAsyncActionCurrentPlaylistReady|PlaylistDataReadyAsync
      (:OnReady
        (if (not (ScalableFloat|GetValueasBool (Variables|Hotfix|GetObjectiveMarkingEnabled)))
          (Components|DestroyComponent))))))

(event EventEndPlay
  (params (EndPlayReason :type EEndPlayReason%20Enum :default "Destroyed"))
  (bind _returnvalue (Components|GetOwner))
  (bind _returnvalue_2 (Actor|GetComponentbyClass _returnvalue "/Script/FortniteGame.AthenaMarkerComponent"))
  (bind _markobjectiveactorsinfomap (Variables|Transient|GetMarkObjectiveActorsInfoMap))
  ;[0.1302 0.07706 0.00983] EndPlay: Unregister all actors and cancel all the markers that were setup for the registered objectives.
  (|CallParentFunction EndPlayReason)
  (if (and (Networking|IsDedicatedServer) (not (ScalableFloat|GetValueasBool (Variables|Hotfix|GetObjectiveMarkingEnabled))))
    (bind _keys (Utilities|Map|Keys _markobjectiveactorsinfomap))
    (for _array_element _keys
      (Utilities|FlowControl|Sequence
        (:then_0
          (Utilities|IsValid _returnvalue
            (:Is%20Valid
              (Utilities|IsValid _returnvalue_2
                (:Is%20Valid
                  (Utilities|IsValid (MarkerComponent|FindMarkerbyID _returnvalue_2 (.MarkerID_36_92E7FF2940604978D0E441BE75F45EE2 (Utilities|Struct|BreakMMObjObjectiveMarkerInfoStruct (Utilities|Map|Find _markobjectiveactorsinfomap _array_element))))
                    (:Is%20Valid
                      (Class|FortWorldMarker|CancelMarker (MarkerComponent|FindMarkerbyID _returnvalue_2 (.MarkerID_36_92E7FF2940604978D0E441BE75F45EE2 (Utilities|Struct|BreakMMObjObjectiveMarkerInfoStruct (Utilities|Map|Find _markobjectiveactorsinfomap _array_element))))))))))))
        (:then_1
          (Public|UnregisterObjective _array_element))))))

(fn GetActorDistanceFromStormPerimeter
  (params (Actor :type Actor%20Object%20Reference))
  (bind _returnvalue (SafeZone|GetSafeZoneIndicator))
  ;[0.009202 0.08854 0.08678] Returns the distance that the actor is from the storm's perimeter? Returns negative value if the actor is inside the storm.
  (Utilities|IsValid Actor
    (:Is%20Valid
      (Utilities|IsValid _returnvalue
        (:Is%20Valid
          (return (- (Math|Vector|Distance2D(Vector) (SafeZone|GetSafeZoneCenter _returnvalue) (Transformation|GetActorLocation Actor)) (SafeZone|GetSafeZoneRadius _returnvalue))))))))

(fn MarkActor
  (params (InActor :type Actor%20Object%20Reference))
  (bind _markobjectiveactorsinfomap (Variables|Transient|GetMarkObjectiveActorsInfoMap))
  (bind (_value _returnvalue) (Utilities|Map|Find _markobjectiveactorsinfomap InActor))
  (bind (_markfrequnecy_23_9396cf96491030a734e4d7ac9f77d57a _markdistancesq_27_800e90424cd7c2d710d2f493e83988ca _maxmarkcount_26_c46e92db43b1c35641a947b346b1541b _includesquad_39_ddbc30f6495f16822544f89be4bae580 _usehoveredmarkerdetail_40_84dad29240fe358bb115a6b928780ef4 _nextservertimeformarking_30_aeff192142020035400de1b8d120323d _currentmarkcount_33_95cd95334038f12f4d0265850f8b7924 _markerid_36_92e7ff2940604978d0e441be75f45ee2) (Utilities|Struct|BreakMMObjObjectiveMarkerInfoStruct _value))
  (bind _returnvalue_2 (Components|GetOwner))
  (bind _returnvalue_3 (Actor|GetComponentbyClass _returnvalue_2 "/Script/FortniteGame.AthenaMarkerComponent"))
  ;[0.07292 0.05884 0.01867] MarkActor: Mark the input actor for the owning player.
  (Utilities|FlowControl|Sequence
    (:then_0
      (if (Networking|IsDedicatedServer)
        (Utilities|IsValid InActor
          (:Is%20Valid
            (if _returnvalue
              (Utilities|IsValid _returnvalue_2
                (:Is%20Valid
                  (Utilities|IsValid _returnvalue_3
                    (:Is%20Valid
                      (Utilities|IsValid (MarkerComponent|FindMarkerbyID _returnvalue_3 _markerid_36_92e7ff2940604978d0e441be75f45ee2)
                        (:Is%20Not%20Valid
                          (bind _returnvalue_4 (MarkerComponent|MarkActoronServer _returnvalue_3 InActor _includesquad_39_ddbc30f6495f16822544f89be4bae580 _usehoveredmarkerdetail_40_84dad29240fe358bb115a6b928780ef4))
                          (Utilities|IsValid (MarkerComponent|FindMarkerbyID _returnvalue_3 _returnvalue_4)
                            (:Is%20Valid
                              (return true _returnvalue_4))))) ; Only mark the actor if it hasn't been marked yet.
                    )))))))))
    (:then_1
      (return false 0))))

(fn RegisteredActorDestroyed
  (params (DestroyedActor :type Actor%20Object%20Reference))
  ;[0.03895 0.009177 0.05729] RegistereActorDestroyed: IF the registered objective actor has been destroyed, unregister it.
  (if (Networking|IsDedicatedServer)
    (Public|UnregisterObjective DestroyedActor)))

(fn RegisterObjective
  (params (MarkActor :type Actor%20Object%20Reference) (InMarkerInfo :type MMObj%20Objective%20Marker%20Hotfix%20Info%20Struct%20Structure))
  (bind (_markingenabled_37_e9c710f04f16fc54efb91c86c40d0196 _markfrequnecy_27_9396cf96491030a734e4d7ac9f77d57a _markdistance_28_800e90424cd7c2d710d2f493e83988ca _maxmarkcount_30_c46e92db43b1c35641a947b346b1541b _includesquad_34_374997384aba530512d186b0137105f9 _usehoveredmarkerdetail_35_1dd554b84ad257830f7aab9a215d2479 _maxownerdistancefromstormperimeter_41_00acfa404b3192b143310790f16c53ac) (Utilities|Struct|BreakMMObjObjectiveMarkerHotfixInfoStruct InMarkerInfo))
  (bind (_value _returnvalue) (Utilities|Map|Find (Variables|Transient|GetMarkObjectiveActorsInfoMap) MarkActor))
  (bind _markactor (Variables|GetMarkActor))
  ;[0.1302 0.02427 0.1129] RegisterObjective: Register the input actor as an objective.
  (if (Networking|IsDedicatedServer)
    (Utilities|IsValid MarkActor
      (:Is%20Valid
        (Utilities|Casting|CastToFortMarkableActorInterface MarkActor
          (:then
            (if (not _returnvalue)
              (bind _output_get (LocalVariables|SetMarkerInfo (Utilities|Struct|MakeMMObjObjectiveMarkerInfoStruct (ScalableFloat|GetValueatLevel _markfrequnecy_27_9396cf96491030a734e4d7ac9f77d57a) (Math|Float|Square (ScalableFloat|GetValueatLevel _markdistance_28_800e90424cd7c2d710d2f493e83988ca)) (ScalableFloat|GetValueasInteger _maxmarkcount_30_c46e92db43b1c35641a947b346b1541b) (ScalableFloat|GetValueasBool _includesquad_34_374997384aba530512d186b0137105f9) (ScalableFloat|GetValueasBool _usehoveredmarkerdetail_35_1dd554b84ad257830f7aab9a215d2479) 0.0 0 "(PlayerID=-1,InstanceID=-1)" (ScalableFloat|GetValueatLevel _maxownerdistancefromstormperimeter_41_00acfa404b3192b143310790f16c53ac))))
              (if (Math|Boolean|ANDBoolean (> (Utilities|Struct|BreakMMObjObjectiveMarkerInfoStruct _output_get) 0) (> (.MarkDistanceSq_27_800E90424CD7C2D710D2F493E83988CA (Utilities|Struct|BreakMMObjObjectiveMarkerInfoStruct _output_get)) 0) (> (.MaxMarkCount_26_C46E92DB43B1C35641A947B346B1541B (Utilities|Struct|BreakMMObjObjectiveMarkerInfoStruct _output_get)) 0))
                (Utilities|Map|Add (Variables|Transient|GetMarkObjectiveActorsInfoMap) _markactor (LocalVariables|GetMarkerInfo))
                (Game|BindEventtoOnDestroyed _markactor (EventDispatchers|CreateEvent :!event RegisteredActorDestroyed))
                (if (== (Utilities|Map|Length (Variables|Transient|GetMarkObjectiveActorsInfoMap)) 1)
                  (Protected|StartUpdateMarking)))
              (else
                (Fortnite|FortShippingLog self (Utilities|String|Append "RegisterObjective: Input actor " (Utilities|GetDisplayName MarkActor) " is already registered for player " (Fortnite|GetPlayerNameSafe (Components|GetOwner))) true))))
          (:CastFailed
            (Fortnite|FortShippingLog self (Utilities|String|Append "RegisterObjective: Input actor does not implement FortMarkableActorInterface: " (Utilities|GetDisplayName MarkActor)) true)))))))

(fn StartUpdateMarking
  ;[0.1771 0.1529 0.0556] StartUpdatingMarking: Start the timer that is gonna periodically run the UpdateMarking() function.
  (if (Networking|IsDedicatedServer)
    (if (not (Utilities|Time|IsTimerActivebyHandle (Variables|Transient|GetUpdateMarkingTimerHandle)))
      (bind _returnvalue (Utilities|Time|SetTimerbyEvent (EventDispatchers|CreateEvent :!event UpdateMarking) (Math|Float|Max(Float) (ScalableFloat|GetValueatLevel (Variables|Hotfix|GetObjectiveMarkingUpdateFrequency)) 1.0) true))
      (Variables|Transient|SetUpdateMarkingTimerHandle _returnvalue)
      (Protected|UpdateMarking))))

(fn StopUpdateMarking
  ;[0.006699 0.03306 0.04688] StopUpdateMarking: Stop the timer that is calling UpdateMarking().
  (if (Networking|IsDedicatedServer)
    (Utilities|Time|ClearandInvalidateTimerbyHandle (Variables|Transient|GetUpdateMarkingTimerHandle))))

(fn UnregisterObjective
  (params (Actor :type Actor%20Object%20Reference))
  ;[0.01521 0.06231 0.06771] UnregisterObjective: Remove the actor from the marked actor map.
  (if (Networking|IsDedicatedServer)
    (Utilities|IsValid Actor
      (:Is%20Valid
        (Utilities|Map|Remove (Variables|Transient|GetMarkObjectiveActorsInfoMap) Actor)
        (Game|UnbindEventfromOnDestroyed Actor (EventDispatchers|CreateEvent :!event RegisteredActorDestroyed)) ; Do not care about the actor's destruction anymore.
        ;;[0.05915 0.06585 0.1198] Stop the marking update timer if there are no more objective actors registered.
        (if (Utilities|Map|IsEmpty (Variables|Transient|GetMarkObjectiveActorsInfoMap))
          (Protected|StopUpdateMarking))))))

(fn UpdateMarking
  (bind _markobjectiveactorsinfomap (Variables|Transient|GetMarkObjectiveActorsInfoMap))
  ;[0.0317 0.09465 0.1458] UpdateMarking: Update the marking of registered objective actors based on the various conditions.
  (if (Networking|IsDedicatedServer)
    (bind _keys (Utilities|Map|Keys (Variables|Transient|GetMarkObjectiveActorsInfoMap)))
    (for _array_element _keys
      (Utilities|IsValid _array_element
        (:Is%20Valid
          (if (.ReturnValue (Utilities|Map|Find _markobjectiveactorsinfomap _array_element))
            (bind _asfort_player_controller_athena (Utilities|Casting|CastToFortPlayerControllerAthena (Components|GetOwner)))
            (Utilities|IsValid (Game|Player|GetPlayerPawn _asfort_player_controller_athena)
              (:Is%20Valid
                (if (> (Math|Vector|DistanceSquared(Vector) (Transformation|GetActorLocation (Game|Player|GetPlayerPawn _asfort_player_controller_athena)) (Transformation|GetActorLocation _array_element)) (Utilities|Struct|BreakMMObjObjectiveMarkerInfoStruct (Utilities|Map|Find _markobjectiveactorsinfomap _array_element)))
                  ;;[0.01392 0.09008 0.09896] Remove marker on an objective actor if it is beyond the marking distance.
                  (Utilities|IsValid (Actor|GetComponentbyClass _asfort_player_controller_athena "/Script/FortniteGame.AthenaMarkerComponent")
                    (:Is%20Valid
                      (Utilities|IsValid (MarkerComponent|FindMarkerbyID (Actor|GetComponentbyClass _asfort_player_controller_athena "/Script/FortniteGame.AthenaMarkerComponent") (.MarkerID_36_92E7FF2940604978D0E441BE75F45EE2 (Utilities|Struct|BreakMMObjObjectiveMarkerInfoStruct (Utilities|Map|Find _markobjectiveactorsinfomap _array_element))))
                        (:Is%20Valid
                          (Class|FortWorldMarker|CancelMarker (MarkerComponent|FindMarkerbyID (Actor|GetComponentbyClass _asfort_player_controller_athena "/Script/FortniteGame.AthenaMarkerComponent") (.MarkerID_36_92E7FF2940604978D0E441BE75F45EE2 (Utilities|Struct|BreakMMObjObjectiveMarkerInfoStruct (Utilities|Map|Find _markobjectiveactorsinfomap _array_element)))))))))
                  (else
                    (bind _distance (Protected|GetActorDistancefromStormPerimeter _array_element))
                    (if (< _distance (.MaxOwnerDistanceFromStormPerimeter_45_D34C34A8488FCA7856756C851EC7BF52 (Utilities|Struct|BreakMMObjObjectiveMarkerInfoStruct (Utilities|Map|Find _markobjectiveactorsinfomap _array_element))))
                      (if (> (Fortnite|Sync|Time|GetServerWorldTimeSeconds) (.NextServerTimeForMarking_30_AEFF192142020035400DE1B8D120323D (Utilities|Struct|BreakMMObjObjectiveMarkerInfoStruct (Utilities|Map|Find _markobjectiveactorsinfomap _array_element))))
                        (bind (_bsuccess _markerid) (Protected|MarkActor _array_element)) ; Mark the actor for the owning player.
                        (if _bsuccess
                          ;;[0.09161 0.02971 0.1094] IF the actor was marked successfully, update the values in the info struct.
                          (bind _structout (Utilities|Struct|SetmembersinMMObjObjectiveMarkerInfoStruct (Utilities|Map|Find (Variables|Transient|GetMarkObjectiveActorsInfoMap) _array_element) (+ (Fortnite|Sync|Time|GetServerWorldTimeSeconds) (Utilities|Struct|BreakMMObjObjectiveMarkerInfoStruct (Utilities|Map|Find (Variables|Transient|GetMarkObjectiveActorsInfoMap) _array_element))) (+ (.CurrentMarkCount_33_95CD95334038F12F4D0265850F8B7924 (Utilities|Struct|BreakMMObjObjectiveMarkerInfoStruct (Utilities|Map|Find (Variables|Transient|GetMarkObjectiveActorsInfoMap) _array_element))) 1) _markerid)) ; Set the next server time for marking attempt.\nIncrement the MarkCount.\nUpdate the MakerID.
                          (Utilities|Map|Add (Variables|Transient|GetMarkObjectiveActorsInfoMap) _array_element _structout)
                          ;;[0.1028 0.009709 0.125] Unregister the objective IF the max mark count has been reached.
                          (if (<= (.MaxMarkCount_26_C46E92DB43B1C35641A947B346B1541B (Utilities|Struct|BreakMMObjObjectiveMarkerInfoStruct _structout)) (.CurrentMarkCount_33_95CD95334038F12F4D0265850F8B7924 (Utilities|Struct|BreakMMObjObjectiveMarkerInfoStruct _structout)))
                            (Public|UnregisterObjective _array_element))))
                      (else
                        (Public|UnregisterObjective _array_element)))))))))))))
