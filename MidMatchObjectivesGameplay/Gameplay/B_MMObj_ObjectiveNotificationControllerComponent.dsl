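; Blueprint component that relays mid-match objective notifications to the notification widget.
; A custom BroadcastNotification event takes a gameplay tag and calls the OnBroadcastNotification event dispatcher with it.
; The notification widget listens to that delegate, so the component only passes the tag through.
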(event Custom|BroadcastNotification
  (params (Tag :type Gameplay%20Tag%20Structure))
  ;[0.04943 0.1094 0.1006] BroadcastNotification: Fires the delegate that gets picked up by the notification widget.
  (Default|CallOnBroadcastNotification Tag))
