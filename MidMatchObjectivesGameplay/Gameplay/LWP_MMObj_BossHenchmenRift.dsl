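; LWP_MMObj_BossHenchmenRift is a Blueprint for a boss henchmen rift, filed under mid-match objectives gameplay.
; It extends its parent's PostSpawnEvents handler by broadcasting an OnSpawn delegate, on the network authority only.
; The construction script only calls the parent, so all other behavior is inherited.
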(event EventPostSpawnEvents
  ;[0.01917 0.1562 0.05894] PostSpawnEvents: Broadcast delegate.
  (|CallParentFunction)
  (Networking|SwitchHasAuthority
    (:Authority
      (Default|CallOnSpawn))))

(fn ConstructionScript
  (|CallParentFunction))
